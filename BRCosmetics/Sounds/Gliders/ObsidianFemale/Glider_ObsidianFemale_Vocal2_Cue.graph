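# Sound cue that plays vocal one-shots for the ObsidianFemale glider, using the glider open/close attenuation and the glider landing sound class.
# An equally weighted Random node picks among four vocal waves and five empty slots, so more than half of triggers are silent.
# Playback runs at reduced volume, with a concurrency override that allows two instances and prevents new ones beyond that.
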SoundCue /BRCosmetics/Sounds/Gliders/ObsidianFemale/Glider_ObsidianFemale_Vocal2_Cue
root=Random_1
nodes [
 {
  id=Random_1
  children [WavePlayer_0,WavePlayer_4,WavePlayer_5,WavePlayer_6,null,null,null,null,null]
  properties {weights [1,1,1,1,1,1,1,1,1]}
 }
 {id=WavePlayer_0, properties {soundWaveAssetPtr=&/Glider_ObsidianFemale_Vocal_01}}
 {id=WavePlayer_4, properties {soundWaveAssetPtr=&/Glider_ObsidianFemale_Vocal_02}}
 {id=WavePlayer_5, properties {soundWaveAssetPtr=&/Glider_ObsidianFemale_Vocal_03}}
 {id=WavePlayer_6, properties {soundWaveAssetPtr=&/Glider_ObsidianFemale_Vocal_04}}
]
properties {
 attenuationSettings=/Game/Sounds/Attenuation/Fort_Player_Glider_OpenClose_Athena_Att
 bOverrideConcurrency=true
 maxDistance=10000
 soundClassObject=&Fort_Gliders_Land
 volumeMultiplier=0.4
 concurrencyOverrides {maxCount=2, resolutionRule=PreventNew}
}
