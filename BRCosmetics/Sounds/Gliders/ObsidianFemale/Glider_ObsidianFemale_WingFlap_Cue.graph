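# Sound cue for the wing flap of the ObsidianFemale glider cosmetic.
# A Random node picks one of six equally weighted flap waves, and a root Modulator varies its playback slightly for variety.
# The cue plays at a reduced volume multiplier through the shared glider thrust attenuation and concurrency assets.
# Its sound class is Fort_Gliders_Land, despite the thrust-oriented attenuation and concurrency.
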SoundCue /BRCosmetics/Sounds/Gliders/ObsidianFemale/Glider_ObsidianFemale_WingFlap_Cue
root=Modulator_0
nodes [
 {id=Modulator_0, children [Random_1], properties {volumeMax=1, volumeMin=0.85}}
 {
  id=Random_1
  children [WavePlayer_1, WavePlayer_5, WavePlayer_6, WavePlayer_7, WavePlayer_10, WavePlayer_11]
  properties {weights [1,1,1,1,1,1]}
 }
 {id=WavePlayer_1, properties {soundWaveAssetPtr=&/Glider_ObsidianFemale_WingFlap_01}}
 {id=WavePlayer_10, properties {soundWaveAssetPtr=&/Glider_ObsidianFemale_WingFlap_05}}
 {id=WavePlayer_11, properties {soundWaveAssetPtr=&/Glider_ObsidianFemale_WingFlap_06}}
 {id=WavePlayer_5, properties {soundWaveAssetPtr=&/Glider_ObsidianFemale_WingFlap_02}}
 {id=WavePlayer_6, properties {soundWaveAssetPtr=&/Glider_ObsidianFemale_WingFlap_03}}
 {id=WavePlayer_7, properties {soundWaveAssetPtr=&/Glider_ObsidianFemale_WingFlap_04}}
]
properties {
 attenuationSettings=&Fort_Player_Glider_Thrust_Athena_Att
 maxDistance=12500
 soundClassObject=&Fort_Gliders_Land
 volumeMultiplier=0.17
 concurrencySet [&Athena_Glider_Thrust_Conc]
}
